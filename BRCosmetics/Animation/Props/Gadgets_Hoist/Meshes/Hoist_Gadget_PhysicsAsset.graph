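# Physics asset for the hoist gadget prop mesh used in emote cosmetics.
# It holds a single body on the Emote_Prop_Hoist_LOD0 bone, shaped as one long, thin capsule rotated off its default axis.
# There are no constraints, so the prop simulates or collides as one rigid piece.
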PhysicsAsset /BRCosmetics/Animation/Props/Gadgets_Hoist/Meshes/Hoist_Gadget_PhysicsAsset
bodies {Emote_Prop_Hoist_LOD0 [{shapeType=Capsule, length=80.9671, radius=1.5753, center=<0 -0.0019 -0.0025>, rotation=<rot 0 -90 -90>}]}
constraints []
